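# Physics asset for the Squad Plaid backpack cosmetic, giving its canister, gun, ribbon and two straps bodies for secondary motion.
# Each part has a tiny "_mstr" box body, constrained to the central pack capsule with free swing and twist, and a larger shape constrained to that master.
# The gun is almost rigid on its master, the canister has slight play, the straps swing widely on one axis, and the ribbon rotates freely.
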PhysicsAsset /BRCosmetics/Accessories/FORT_Backpacks/M_MED_COD_Squad_Plaid_Backpack/Meshes/M_MED_COD_Squad_Plaid_Pack_PhysicsAsset
bodies {
 dyn_cannister [{shapeType=Capsule, length=15, radius=3.5, center=<-3.5141 0 4.8974>, rotation=<rot 90 0 0>}]
 dyn_cannister_mstr [{shapeType=Box, extentX=1.01, extentY=1.01, extentZ=1.01}]
 dyn_gun [{shapeType=Box, extentX=15.5112, extentY=6.0434, extentZ=25.3663, center=<4.7725 3.0914 0.2582>}]
 dyn_gun_mstr [{shapeType=Box, extentX=1.01, extentY=1.01, extentZ=1.01}]
 dyn_ribbon [{shapeType=Box, extentX=11.421, extentY=4.8224, extentZ=2.9415, center=<5.6339 -0.4885 0.7077>}]
 dyn_ribbon_mstr [{shapeType=Box, extentX=1.01, extentY=1.01, extentZ=1.01}]
 dyn_strap_a [{shapeType=Box, extentX=4, extentY=2, extentZ=4, center=<2 0 0>}]
 dyn_strap_a_mstr [{shapeType=Box, extentX=1.01, extentY=1.01, extentZ=1.01}]
 dyn_strap_b [{shapeType=Box, extentX=4, extentY=2, extentZ=4, center=<2 0 0>}]
 dyn_strap_b_mstr [{shapeType=Box, extentX=1.01, extentY=1.01, extentZ=1.01}]
 pack [{shapeType=Capsule, length=13.8102, radius=10.8408, center=<-0.0662 -7.4571 -5.4728>, rotation=<rot 0 75.2638 -58.1024>}]
}
constraints [
 {
  bone1Name=dyn_cannister_mstr
  bone2Name=pack
  swing1LimitDegrees=45
  swing1Motion=Free
  swing2LimitDegrees=45
  swing2Motion=Free
  twistLimitDegrees=45
  twistMotion=Free
 }
 {
  bone1Name=dyn_strap_b_mstr
  bone2Name=pack
  swing1LimitDegrees=45
  swing1Motion=Free
  swing2LimitDegrees=45
  swing2Motion=Free
  twistLimitDegrees=45
  twistMotion=Free
 }
 {
  bone1Name=dyn_strap_a_mstr
  bone2Name=pack
  swing1LimitDegrees=45
  swing1Motion=Free
  swing2LimitDegrees=45
  swing2Motion=Free
  twistLimitDegrees=45
  twistMotion=Free
 }
 {
  bone1Name=dyn_gun_mstr
  bone2Name=pack
  swing1LimitDegrees=45
  swing1Motion=Free
  swing2LimitDegrees=45
  swing2Motion=Free
  twistLimitDegrees=45
  twistMotion=Free
 }
 {
  bone1Name=dyn_ribbon_mstr
  bone2Name=pack
  swing1LimitDegrees=45
  swing1Motion=Free
  swing2LimitDegrees=45
  swing2Motion=Free
  twistLimitDegrees=45
  twistMotion=Free
 }
 {bone1Name=dyn_gun, bone2Name=dyn_gun_mstr, swing1LimitDegrees=1, swing2LimitDegrees=1, twistLimitDegrees=0.1}
 {bone1Name=dyn_strap_a, bone2Name=dyn_strap_a_mstr, swing1LimitDegrees=63.8571, swing2LimitDegrees=5, twistLimitDegrees=10}
 {bone1Name=dyn_strap_b, bone2Name=dyn_strap_b_mstr, swing1LimitDegrees=77.5714, swing2LimitDegrees=5, twistLimitDegrees=10}
 {
  bone1Name=dyn_ribbon
  bone2Name=dyn_ribbon_mstr
  swing1LimitDegrees=45
  swing1Motion=Free
  swing2LimitDegrees=45
  swing2Motion=Free
  twistLimitDegrees=45
  twistMotion=Free
 }
 {bone1Name=dyn_cannister, bone2Name=dyn_cannister_mstr, swing1LimitDegrees=10, swing2LimitDegrees=15, twistLimitDegrees=5}
]
